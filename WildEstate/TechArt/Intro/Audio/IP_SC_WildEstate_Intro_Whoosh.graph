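# Sound cue for the WildEstate intro whoosh, built around a wave named as a Battle Bus intro transition.
# A single Wave Player plays that one sound directly, with no randomization, modulation or mixing nodes.
# The cue uses the Fort_Gadgets sound class with bus sends disabled.
# Its concurrency override allows one instance at a time and prevents new plays while one is active.
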SoundCue /WildEstate/TechArt/Intro/Audio/IP_SC_WildEstate_Intro_Whoosh
root=WavePlayer_0
nodes [{id=WavePlayer_0, properties {soundWaveAssetPtr=&/IP_WildEstate_BattleBus_Intro_Transition_Whoosh_01}}]
properties {
 bEnableBusSends=false
 bOverrideConcurrency=true
 maxDistance=2097152
 soundClassObject=&Fort_Gadgets
 volumeMultiplier=1
 concurrencyOverrides {maxCount=1, resolutionRule=PreventNew}
}
