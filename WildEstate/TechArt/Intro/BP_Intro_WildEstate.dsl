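; Client-only actor that plays the WildEstate intro overlay during the Battle Royale Aircraft phase, destroying itself afterwards or where cosmetics cannot run.
; A game phase event listener starts rendering, tick and an audio control bus mix.
; Two show-only scene captures render clouds, text, background and text mask into runtime render targets feeding a post-process blendable material.
; Tick drives the animation through a normalized Progress parameter in a Material Parameter Collection.
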(event EventBeginPlay
  (bind _returnvalue (BattleRoyaleGamePhaseLogic|GetBattleRoyaleGamePhaseLogicComponent))
  (|CallParentFunction)
  (Networking|CanExecuteCosmeticEvents
    (:True
      (Utilities|IsValid _returnvalue
        (:Is%20Valid
          (switch Utilities|FlowControl|Switch|SwitchonEAthenaGamePhase (BattleRoyaleGamePhaseLogic|GetGamePhase _returnvalue)
            (:None
              _)
            (:Setup
              _)
            (:Warmup
              _)
            (:Aircraft
              _)
            (:SafeZones
              ;[1 1 1] Destroy this actor on the Server.
              (Actor|DestroyActor :!label _label))
            (:EndGame
              (goto _label)))
          ;[1 1 1] Client: listen to GamePhaseUpdatedEvent in order to respond to Aircraft phase beginning.
          (bind (_asynctaskproxy _payload) (GameplayEventRouter|StartListeningToStatefulEvent (GameplayEventRouter|GetGlobalEventRouter) "/Script/FortniteGame.FortGamePhaseUpdatedEvent")
            (:then
              (Variables|Default|SetGamePhaseChangedTaskRef _asynctaskproxy))
            (:OnEventReceived
              (CallFunction|OnGamePhaseChanged (Utilities|Struct|BreakFortGamePhaseUpdatedEvent _payload)))))
        (:Is%20Not%20Valid
          (Development|PrintString "BeginPlay: cannot get GamePhase.  Not listening to stateful events." true true "(R=1.000000,G=0.000000,B=0.000000,A=1.000000)" 5.0))))
    (:False
      ;[1 1 1] Destroy this actor on the Server.
      (Actor|DestroyActor))))

(event Custom|InitializeRendering
  (bind _returnvalue (BattleRoyaleGamePhaseLogic|GetBattleRoyaleGamePhaseLogicComponent))
  (Networking|CanExecuteCosmeticEvents
    (:True
      (Utilities|IsValid _returnvalue
        (:Is%20Valid
          (switch Utilities|FlowControl|Switch|SwitchonEAthenaGamePhase (BattleRoyaleGamePhaseLogic|GetGamePhase _returnvalue)
            (:None
              _)
            (:Setup
              _)
            (:Warmup
              _)
            (:Aircraft
              _)
            (:SafeZones
              ;[1 1 1] Destroy this actor on the Server.
              (Actor|DestroyActor :!label _label))
            (:EndGame
              (goto _label)))
          ;[1 1 1] Client: listen to GamePhaseUpdatedEvent in order to respond to Aircraft phase beginning.
          (bind (_asynctaskproxy _payload) (GameplayEventRouter|StartListeningToStatefulEvent (GameplayEventRouter|GetGlobalEventRouter) "/Script/FortniteGame.FortGamePhaseUpdatedEvent")
            (:then
              (Variables|Default|SetGamePhaseChangedTaskRef _asynctaskproxy))
            (:OnEventReceived
              (CallFunction|OnGamePhaseChanged (Utilities|Struct|BreakFortGamePhaseUpdatedEvent _payload)))))
        (:Is%20Not%20Valid
          (Development|PrintString "BeginPlay: cannot get GamePhase.  Not listening to stateful events." true true "(R=1.000000,G=0.000000,B=0.000000,A=1.000000)" 5.0))))
    (:False
      ;[1 1 1] Destroy this actor on the Server.
      (Actor|DestroyActor))))

(event EventEndPlay
  (params (EndPlayReason :type EEndPlayReason%20Enum :default "Destroyed"))
  (|CallParentFunction EndPlayReason)
  ;[1 1 1] Deactivate audio modulation
  (Networking|CanExecuteCosmeticEvents
    (:True
      (Audio|Modulation|DeactivateControlBusMix (Variables|Default|GetIntroControlBusMix)))
    (:False
      _))
  (bind _gamephasechangedtaskref (Variables|Default|GetGamePhaseChangedTaskRef)
    (:then
      (GameplayEventRouter|StopListeningtoStatefulEvent _gamephasechangedtaskref))
    (:else)))

(event Custom|OnGamePhaseChanged
  (params (GamePhase :type EAthenaGamePhase%20Enum))
  (switch Utilities|FlowControl|Switch|SwitchonEAthenaGamePhase GamePhase
    (:None
      _)
    (:Setup
      _)
    (:Warmup
      _)
    (:Aircraft
      ;[1 1 1] Start rendering the intro sequence, enable tick, start the intro sequence timer
      (Development|PrintString "OnGamePhaseChanged: Starting WildEstate into." true true "(R=0.000000,G=0.660000,B=1.000000,A=1.000000)" 5.0)
      (CallFunction|InitializeRendering)
      (Actor|Tick|SetActorTickEnabled true)
      (Variables|Default|SetIntroSequenceCurrentSeconds 0.0)
      ;[1 1 1] Activates audio modulation
      (Audio|Modulation|ActivateControlBusMix (Variables|Default|GetIntroControlBusMix)))
    (:SafeZones
      ;[1 1 1] Destroy ourselves if we've moved past Aircraft phase
      (Actor|DestroyActor :!label _label))
    (:EndGame
      (goto _label)))
  ;[1 1 1] Make sure we're not ticking unless we're running the intro
  (Actor|Tick|SetActorTickEnabled false))

(event EventTick
  (params (DeltaSeconds :type Float%20%28single-precision%29))
  (bind _introsequencedurationseconds (Variables|Default|GetIntroSequenceDurationSeconds))
  (|CallParentFunction DeltaSeconds)
  (if (<= (Variables|Default|GetIntroSequenceCurrentSeconds) (Variables|Default|GetIntroSequenceDurationSeconds))
    ;[1 1 1] Calculate our progress between 0-1 and pass this into the material to drive the animation.
    (bind _output_get (Variables|Default|SetIntroSequenceCurrentSeconds (+ (Variables|Default|GetIntroSequenceCurrentSeconds) DeltaSeconds)))
    (Development|PrintString (Utilities|String|Append "EventTick: Intro Progress - " (Utilities|String|ToString(Float) (Utilities|ClampMaxFloat (/ _output_get _introsequencedurationseconds) 1.0))) true true "(R=0.000000,G=0.660000,B=1.000000,A=1.000000)" 0.0)
    (Rendering|Material|SetScalarParameterValue "/WildEstate/TechArt/Intro/MPC_Intro_WildEstate.MPC_Intro_WildEstate" "Progress" (Utilities|ClampMaxFloat (/ _output_get _introsequencedurationseconds) 1.0) :!class /Script/Engine.KismetMaterialLibrary) ; This needs to be a global parameter
    (else
      ;[1 1 1] Intro completed, disable tick and destroy
      (Development|PrintString "Intro complete." true true "(R=0.000000,G=0.660000,B=1.000000,A=1.000000)" 0.0)
      (Actor|Tick|SetActorTickEnabled false)
      (Variables|Default|SetIntroSequenceCurrentSeconds 1.0)
      (Actor|DestroyActor))))

(fn InitializeRendering
  (Utilities|FlowControl|Sequence
    (:then_0
      (bind _returnvalue (CanvasRenderTarget2D|CreateCanvasRenderTarget2D "/Script/Engine.CanvasRenderTarget2D" 32 32))
      (Variables|Default|SetIntroSequenceRenderTargetMain _returnvalue)
      (bind _returnvalue_2 (CanvasRenderTarget2D|CreateCanvasRenderTarget2D "/Script/Engine.CanvasRenderTarget2D" 32 32))
      (Variables|Default|SetIntroSequenceRenderTargetMask _returnvalue_2))
    (:then_1
      (Rendering|SceneCapture|ClearShowOnlyComponents (Variables|Default|GetSceneCaptureComponent2D_Main))
      (for _array_element (Utilities|Array|MakeArray (Variables|Default|GetSM_WildEstate_Cloud) (Variables|Default|GetSM_WildEstate_Cloud1) (Variables|Default|GetSM_WildEstate_Cloud2) (Variables|Default|GetSM_WildEstate_Cloud3) (Variables|Default|GetSM_WildEstate_Cloud4) (Variables|Default|GetSM_WildEstate_Cloud5) (Variables|Default|GetSM_WildEstate_Cloud6) (Variables|Default|GetSM_WildEstate_Cloud7) (Variables|Default|GetSM_WildEstate_Cloud8) (Variables|Default|GetSM_WildEstate_Cloud9) (Variables|Default|GetIntroText) (Variables|Default|GetBackground))
        (Rendering|SceneCapture|ShowOnlyComponent (Variables|Default|GetSceneCaptureComponent2D_Main) _array_element))
      (Variables|/Script/Engine.SceneCaptureComponent2D|SetTextureTarget (Variables|Default|GetIntroSequenceRenderTargetMain) (Variables|Default|GetSceneCaptureComponent2D_Main)))
    (:then_2
      (Rendering|SceneCapture|ClearShowOnlyComponents (Variables|Default|GetSceneCaptureComponent2D_Mask))
      (Rendering|SceneCapture|ShowOnlyComponent (Variables|Default|GetSceneCaptureComponent2D_Mask) (Variables|Default|GetTextMask))
      (Variables|/Script/Engine.SceneCaptureComponent2D|SetTextureTarget (Variables|Default|GetIntroSequenceRenderTargetMask) (Variables|Default|GetSceneCaptureComponent2D_Mask)))
    (:then_3
      (bind _returnvalue_3 (Rendering|Material|CreateDynamicMaterialInstance (Variables|Default|GetIntroSequencePostProcessMaterialParent) "Intro to Screen" "None" :!class /Script/Engine.KismetMaterialLibrary))
      (Variables|Default|SetIntroSequencePostProcessMaterialInstance _returnvalue_3)
      (Rendering|Material|SetTextureParameterValue (Variables|Default|GetIntroSequencePostProcessMaterialInstance) "SceneCapture" (Variables|Default|GetIntroSequenceRenderTargetMain))
      (Rendering|Material|SetTextureParameterValue (Variables|Default|GetIntroSequencePostProcessMaterialInstance) "SceneCaptureMask" (Variables|Default|GetIntroSequenceRenderTargetMask))
      (Utilities|Struct|SetmembersinPostProcessSettings (Variables|/Script/Engine.PostProcessComponent|GetSettings (Variables|Default|GetPostProcess)) (Utilities|Struct|MakeWeightedBlendables (Utilities|Array|MakeArray (Utilities|Struct|MakeWeightedBlendable 1.0 (Variables|Default|GetIntroSequencePostProcessMaterialInstance))))))))

(fn ConstructionScript)
