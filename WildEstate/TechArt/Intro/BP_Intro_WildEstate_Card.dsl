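; Client-side actor that plays the WildEstate battle bus intro card during the Aircraft phase, with cutoff-limited late-joiner support.
; Once the loading screen hides, it attaches a viewport-fitted card plane to each local player's aircraft, hiding HUD and blocking look input.
; A timeline drives material collection progress, camera modifier offsets and a whoosh sound.
; EndPlay restores everything after completion or aircraft exit; a tick-based test hook can delay the start.
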(event EventBeginPlay
  (bind _returnvalue (BattleRoyaleGamePhaseLogic|GetBattleRoyaleGamePhaseLogicComponent))
  ;[0 1 0 0.5] [BeginPlay]
  (|CallParentFunction)
  (if (not (Variables|Default|GetbDisableIntroCard))
    (Networking|CanExecuteCosmeticEvents
      (:True
        (Utilities|IsValid _returnvalue
          (:Is%20Valid
            (switch Utilities|FlowControl|Switch|SwitchonEAthenaGamePhase (BattleRoyaleGamePhaseLogic|GetGamePhase _returnvalue)
              (:None
                _)
              (:Setup
                _)
              (:Warmup
                _)
              (:Aircraft
                (Fortnite|FortShippingLog self "[BP_Intro_WildEstate_Card] [BeginPlay] Trying to start the intro sequence for late joiner.")
                (CallFunction|BindtoonExitAircraft)
                (CallFunction|TryStartIntroSequenceforLateJoiner))
              (:SafeZones
                (Fortnite|FortShippingLog self "[BP_Intro_WildEstate_Card] [BeginPlay] Match has progressed beyond Aircraft phase.  Destroying actor." true :!label _label)
                (Actor|DestroyActor) ; Destroy self if the match has already passed the Aircraft phase.
              )
              (:EndGame
                (goto _label)))
            (Fortnite|FortShippingLog self "[BP_Intro_WildEstate_Card] [BeginPlay] Listening for GamePhaseUpdatedEvent.")
            ;;[1 1 1] Listen to GamePhaseUpdatedEvent in order to respond to Aircraft phase beginning.
            (bind (_asynctaskproxy _payload) (GameplayEventRouter|StartListeningToStatefulEvent (GameplayEventRouter|GetGlobalEventRouter) "/Script/FortniteGame.FortGamePhaseUpdatedEvent")
              (:then
                (Variables|CachedRefs|SetGamePhaseChangedTaskRef _asynctaskproxy)
                (CallFunction|BindtoonExitAircraft))
              (:OnEventReceived
                (bind _payload_gamephase (Utilities|Struct|BreakFortGamePhaseUpdatedEvent _payload))
                (CallFunction|OnGamePhaseChanged _payload_gamephase))))
          (:Is%20Not%20Valid
            (Fortnite|FortShippingLog self "[BP_Intro_WildEstate_Card] [BeginPlay] Cannot get GamePhase.  Not listening to stateful events." true))))
      (:False
        _)))
  (Actor|DestroyActor) ; Destroy self on the server
)

(event EventEndPlay
  (params (EndPlayReason :type EEndPlayReason%20Enum :default "Destroyed"))
  ;[1 0 0 0.5] [EndPlay]
  (|CallParentFunction EndPlayReason)
  (Networking|CanExecuteCosmeticEvents
    (:True
      (bind _asfort_game_instance (Utilities|Casting|CastToFortGameInstance (Game|GetGameInstance))
        (:then
          (EventDispatchers|UnbindEventfromOnLoadingScreenVisibilityChanged _asfort_game_instance (EventDispatchers|CreateEvent :!event Custom|OnLoadingScreenVisibilityChanged)))
        (:CastFailed
          _))
      (Utilities|FlowControl|Sequence
        (:then_0
          (CallFunction|HandleProgressChanged 1.0)
          (Audio|Modulation|DeactivateControlBusMix (Variables|CachedRefs|GetIntroControlBusMix)) ; Deactivate audio modulation
          (bind _gamephasechangedtaskref (Variables|CachedRefs|GetGamePhaseChangedTaskRef)
            (:then
              (GameplayEventRouter|StopListeningtoStatefulEvent _gamephasechangedtaskref))
            (:else)))
        (:then_1
          (for _array_element (Variables|CachedRefs|GetPlayerViewInfos)
            (Utilities|FlowControl|Sequence
              (:then_0
                (HUD|SetHUDElementVisibilitywithContext (Utilities|Struct|BreakLocalPlayerControllerView _array_element) (Variables|Config|GetHUDElement%20Tags) false (Variables|Config|GetHideHUDReasonName) "") ; Enable HUD and look input for each local player controller
                (Input|SetIgnoreLookInput (Utilities|Struct|BreakLocalPlayerControllerView _array_element) false))
              (:then_1
                (Utilities|IsValid (.PlayerCameraManagerRef_9_B78DA68F490CBF5D4F6166B7A1088195 (Utilities|Struct|BreakLocalPlayerControllerView _array_element))
                  (:Is%20Valid
                    (Utilities|IsValid (.CameraModifierRef_15_4B58C77948E7321B314F7B84B4C31B53 (Utilities|Struct|BreakLocalPlayerControllerView _array_element))
                      (:Is%20Valid
                        (Game|Player|RemoveCameraModifier (.PlayerCameraManagerRef_9_B78DA68F490CBF5D4F6166B7A1088195 (Utilities|Struct|BreakLocalPlayerControllerView _array_element)) (.CameraModifierRef_15_4B58C77948E7321B314F7B84B4C31B53 (Utilities|Struct|BreakLocalPlayerControllerView _array_element))))
                      (:Is%20Not%20Valid
                        (Fortnite|FortShippingLog self "[BP_Intro_WildEstate_Card] [EndPlay] CameraModifier is null" true))))
                  (:Is%20Not%20Valid
                    (Fortnite|FortShippingLog self "[BP_Intro_WildEstate_Card] [EndPlay] PlayerCameraManager is null" true))))
              (:then_2
                (Utilities|IsValid (.IntroCardRef_12_023F7DB34C8B6B566E978D97B57377AB (Utilities|Struct|BreakLocalPlayerControllerView _array_element))
                  (:Is%20Valid
                    (Actor|DestroyActor (.IntroCardRef_12_023F7DB34C8B6B566E978D97B57377AB (Utilities|Struct|BreakLocalPlayerControllerView _array_element))) ; Destroy IntroCard
                  )
                  (:Is%20Not%20Valid
                    (Fortnite|FortShippingLog self "[BP_Intro_WildEstate_Card] [EndPlay] IntroCard is null" true))))))
          (CallFunction|UnbindfromonExitAircraft) ; Make sure to unbind events
        )))))

(event Custom|OnGamePhaseChanged
  (params (GamePhase :type EAthenaGamePhase%20Enum))
  (bind _self self)
  ;[1 1 0 0.5] [OnGamePhaseChanged]
  (switch Utilities|FlowControl|Switch|SwitchonEAthenaGamePhase GamePhase
    (:Aircraft
      (if (>= (Variables|Default|GetDelayedTriggerIntroSequencePercent) 0)
        (Fortnite|FortShippingLog _self "[BP_Intro_WildEstate_Card] [OnGamePhaseChanged] Intro sequence activation delayed.")
        (else
          (Fortnite|FortShippingLog _self "[BP_Intro_WildEstate_Card] [OnGamePhaseChanged] Reached Aircraft phase.  Starting WildEstate intro.")
          (CallFunction|BeginIntroSequence true))))
    (:SafeZones
      _)
    (:EndGame
      _))
  (Fortnite|FortShippingLog self "[BP_Intro_WildEstate_Card] [OnGamePhaseChanged] Match has progressed beyond Aircraft phase.  Destroying actor." true)
  (Actor|DestroyActor) ; Destroy self after Aircraft phase
)

(event Custom|BeginIntroSequence
  (params (bCheckLoadingScreenVisibility :type Boolean))
  ;[1 0.5 0 0.5] [BeginIntroSequence]
  (if bCheckLoadingScreenVisibility
    (bind _asfort_game_instance (Utilities|Casting|CastToFortGameInstance (Game|GetGameInstance))
      (:then
        (if (Class|FortGameInstance|IsCurrentlyShowingLoadingScreen _asfort_game_instance)
          (Fortnite|FortShippingLog self "[BeginIntroSequence] Waiting for loading screen to hide.")
          (EventDispatchers|UnbindEventfromOnLoadingScreenVisibilityChanged _asfort_game_instance (EventDispatchers|CreateEvent :!event Custom|OnLoadingScreenVisibilityChanged))
          (EventDispatchers|BindEventtoOnLoadingScreenVisibilityChanged _asfort_game_instance (EventDispatchers|CreateEvent :!event Custom|OnLoadingScreenVisibilityChanged))
          (else
            _)))
      (:CastFailed
        (Fortnite|FortShippingLog self "[BeginIntroSequence] Failed to find FortGameInstance.  Unable to check for loading screen.  Intro will NOT play!" true))))
  (Utilities|FlowControl|DoOnce false
    (:Completed
      (Fortnite|FortShippingLog self "[BeginIntroSequence] Beginning intro sequnce.")
      (Audio|Modulation|ActivateControlBusMix (Variables|CachedRefs|GetIntroControlBusMix)) ; Activates audio modulation
      (Default|GatherPlayerViewInfo)
      (Utilities|FlowControl|Sequence
        (:then_0
          (for (_array_element _array_index) (Variables|CachedRefs|GetPlayerViewInfos)
            (bind _updatedplayerviewinfo (Default|SetUpIntroCard _array_element))
            (bind _updatedplayerviewinfo_2 (Default|SetUpCameraModifier _updatedplayerviewinfo))
            (HUD|SetHUDElementVisibilitywithContext (Utilities|Struct|BreakLocalPlayerControllerView _updatedplayerviewinfo_2) (Variables|Config|GetHUDElement%20Tags) true (Variables|Config|GetHideHUDReasonName) "")
            (Input|SetIgnoreLookInput (Utilities|Struct|BreakLocalPlayerControllerView _updatedplayerviewinfo_2) true)
            (Utilities|Array|SetArrayElem (Variables|CachedRefs|GetPlayerViewInfos) _array_index _updatedplayerviewinfo_2)))
        (:then_1
          (CallFunction|RunTimeline)
          (Variables|Default|SetbIntroBegun true))))))

(event Custom|OnLoadingScreenVisibilityChanged
  (params (bVisible :type Boolean))
  ;[0.1419 0.03266 0.15 0.5] [OnLoadingScreenVisibilityChanged] When the loading screen hides, now we can begin the intro sequence
  (if (not bVisible)
    (Fortnite|FortShippingLog self "[OnLoadingScreenVisibilityChanged] Visible = false.  Calling BeginIntroSequence." true)
    (bind _asfort_game_instance (Utilities|Casting|CastToFortGameInstance (Game|GetGameInstance))
      (:then
        (EventDispatchers|UnbindEventfromOnLoadingScreenVisibilityChanged _asfort_game_instance (EventDispatchers|CreateEvent :!event Custom|OnLoadingScreenVisibilityChanged)))
      (:CastFailed
        _))
    (CallFunction|BeginIntroSequence false) ; Note that in this case we tell BeginIntroSequence\nto NOT check the loading screen visibility.\nCalling IsCurrentlyShowingLoadingScreen at this time\nwill still return true.
  ))

(event Custom|RunTimeline
  ;[0 0 1 0.5] [RunTimeline]
  (Fortnite|FortShippingLog self "[BP_Intro_WildEstate_Card] [RunTimeline] Starting the intro timeline.")
  (Components|Timeline|SetPlayRate (Variables|BP_Intro_WildEstate_Card|GetTimeline) (Math|Float|SafeDivide (Components|Timeline|GetTimelineLength (Variables|BP_Intro_WildEstate_Card|GetTimeline)) (Variables|Config|GetIntroSequenceDurationSeconds)))
  (bind (_direction _curve_wildestate_buscamloc _curve_wildestate_buscamrot) (Timeline|Timeline PlayFromStart)
    (:Update
      (bind (_curve_wildestate_buscamrot_x _curve_wildestate_buscamrot_y _curve_wildestate_buscamrot_z) (Math|Vector|BreakVector _curve_wildestate_buscamrot))
      (CallFunction|HandleProgressChanged 0 (Utilities|ClampMaxFloat (Math|Float|SafeDivide (Components|Timeline|GetPlaybackPosition (Variables|BP_Intro_WildEstate_Card|GetTimeline)) (Components|Timeline|GetTimelineLength (Variables|BP_Intro_WildEstate_Card|GetTimeline))) 1.0) _curve_wildestate_buscamloc (Math|Rotator|MakeRotator _curve_wildestate_buscamrot_x _curve_wildestate_buscamrot_y _curve_wildestate_buscamrot_z)))
    (:Finished
      (Fortnite|FortShippingLog self "[BP_Intro_WildEstate_Card] [RunTimeline] Timeline complete")
      (Actor|DestroyActor) ; We're done.  All cleanup will take place in EndPlay.
    )
    (:Curve_WildEstate_BusWhooshEvent
      ;;[1 1 1] Play a whoosh sound when the timeline event triggers
      (Audio|PlaySound2D (Variables|Config|GetBusWhooshSoundCue)))))

(event Custom|HandleProgressChanged
  (params (ProgressPercent :type Float%20%28double-precision%29) (CamRelativeLoc :type Vector) (CamRelativeRot :type Rotator))
  ;[1 0 1 0.5] [HandleProgressChanged] Update material and camera modifier with calculated progress.
  (Utilities|FlowControl|Sequence
    (:then_0
      (Rendering|Material|SetScalarParameterValue "/WildEstate/TechArt/Intro/MPC_Intro_WildEstate.MPC_Intro_WildEstate" "Progress" ProgressPercent :!class /Script/Engine.KismetMaterialLibrary) ; Set the intro material's progress percent
      (for _array_element (Variables|CachedRefs|GetPlayerViewInfos)
        (bind (_array_element_playercontrollerref_2_74c1ca9f4ce69ec3722fde9eb6414ff8 _array_element_aircraftref_17_f59a25bd418b6c826bf779b03f5b9e8c _array_element_playercameramanagerref_9_b78da68f490cbf5d4f6166b7a1088195 _array_element_introcardref_12_023f7db34c8b6b566e978d97b57377ab _array_element_cameramodifierref_15_4b58c77948e7321b314f7b84b4c31b53) (Utilities|Struct|BreakLocalPlayerControllerView _array_element))
        (Utilities|IsValid _array_element_cameramodifierref_15_4b58c77948e7321b314f7b84b4c31b53
          (:Is%20Valid
            (Variables|/WildEstate/TechArt/Intro/CameraModifier_WildEstate_Intro.CameraModifier_WildEstate_Intro_C|SetCameraLocationOffsetRelative CamRelativeLoc _array_element_cameramodifierref_15_4b58c77948e7321b314f7b84b4c31b53)
            (Variables|/WildEstate/TechArt/Intro/CameraModifier_WildEstate_Intro.CameraModifier_WildEstate_Intro_C|SetCameraRotationOffsetRelative CamRelativeRot _array_element_cameramodifierref_15_4b58c77948e7321b314f7b84b4c31b53))
          (:Is%20Not%20Valid
            (Development|PrintString "HandleProgressChanged: CameraModifier is null")))))
    (:then_1
      (Development|PrintString (Utilities|String|Append "IntroCard Timeline: Progress %: " (Utilities|String|ToString(Float) (* ProgressPercent 1e+02))) true false "(R=0.000000,G=0.660000,B=1.000000,A=1.000000)" 0.0)
      (Development|PrintString (Utilities|String|Append "IntroCard Timeline: Loc X: " (Utilities|String|ToString(Float) (.x CamRelativeLoc))) true false "(R=0.000000,G=0.660000,B=1.000000,A=1.000000)" 0.0)
      (Development|PrintString (Utilities|String|Append "IntroCard Timeline: Loc Y: " (Utilities|String|ToString(Float) (.y CamRelativeLoc))) true false "(R=0.000000,G=0.660000,B=1.000000,A=1.000000)" 0.0)
      (Development|PrintString (Utilities|String|Append "IntroCard Timeline: Loc Z: " (Utilities|String|ToString(Float) (.z CamRelativeLoc))) true false "(R=0.000000,G=0.660000,B=1.000000,A=1.000000)" 0.0)
      (Development|PrintString (Utilities|String|Append "IntroCard Timeline: Roll: " (Utilities|String|ToString(Float) (.roll CamRelativeRot))) true false "(R=0.000000,G=0.660000,B=1.000000,A=1.000000)" 0.0)
      (Development|PrintString (Utilities|String|Append "IntroCard Timeline: Pitch: " (Utilities|String|ToString(Float) (.pitch CamRelativeRot))) true false "(R=0.000000,G=0.660000,B=1.000000,A=1.000000)" 0.0)
      (Development|PrintString (Utilities|String|Append "IntroCard Timeline: Yaw: " (Utilities|String|ToString(Float) (.yaw CamRelativeRot))) true false "(R=0.000000,G=0.660000,B=1.000000,A=1.000000)" 0.0))))

(event Custom|TryStartIntroSequenceForLateJoiner
  ;[1 1 1 0.5] [TryStartIntroSequenceForLateJoiner]
  (if (Variables|Config|GetbAllowLateJoiners)
    (Fortnite|FortShippingLog self "[BP_Intro_WildEstate_Card] [TryStartIntroSequenceForLateJoiner] Attempting to start the intro sequence for a late joiner.")
    (bind _array_element (Utilities|Array|ForEachLoopwithBreak (Fortnite|GetLocalFortPlayerControllers true) :!label _label)
      (:LoopBody
        (Fortnite|FortShippingLog self (Utilities|String|Append "[BP_Intro_WildEstate_Card] [TryStartIntroSequenceForLateJoiner] Current percent: " (Utilities|String|ToString(Float) (CallFunction|CalculateFlightPathProgressPercent _array_element)) ", Cutoff percent: " (Utilities|String|ToString(Float) (Variables|Config|GetLateJoinerIntroCutoffPercent))))
        (if (and (>= (CallFunction|CalculateFlightPathProgressPercent _array_element) 0) (< (CallFunction|CalculateFlightPathProgressPercent _array_element) (Variables|Config|GetLateJoinerIntroCutoffPercent)))
          (Fortnite|FortShippingLog self "[BP_Intro_WildEstate_Card] [TryStartIntroSequenceForLateJoiner] Starting the intro sequence for late joiner.")
          (Variables|Default|SetbLateJoinSuccess true)
          (CallFunction|BeginIntroSequence true)
          (goto _label Break)
          (else
            (Fortnite|FortShippingLog self "[BP_Intro_WildEstate_Card] [TryStartIntroSequenceForLateJoiner] Not starting intro sequence for late joiner: flight path is too far along." true)
            (goto _label Break))))
      (:Completed
        (if (not (Variables|Default|GetbLateJoinSuccess))
          (Fortnite|FortShippingLog self "[BP_Intro_WildEstate_Card] [TryStartIntroSequenceForLateJoiner] Unable to start intro sequence.  Destroying actor." true)
          (Actor|DestroyActor))))
    (else
      (Fortnite|FortShippingLog self "[BP_Intro_WildEstate_Card] [TryStartIntroSequenceForLateJoiner] Not starting intro sequence for late joiners: bAllowLateJoiners = false.  Destroying actor." true)
      (Actor|DestroyActor))))

(event Custom|OnExitAircraft
  (params (Controller :type Controller%20Object%20Reference))
  ;[1 1 1] [OnExitAircraft] - If the player exits the aircraft, just destroy this component.
  (Actor|DestroyActor) ; We're done.  All cleanup will take place in EndPlay.
)

(event Custom|DelayIntroSequenceTrigger
  (params (IntroTriggerPercent :type Float%20%28double-precision%29))
  ;[0 0 0 0.5] For testing delaying the intro sequence to fake late joiners
  ;;[1 1 1] [DelayIntroSequenceTrigger] - Receives a float representing a percentage of the flight path where the intro will be started.  Meant for testing purposes.
  (Development|PrintString (Utilities|String|Append "[DelayIntroSequenceTrigger] Delaying the activation of the intro sequence until " (Utilities|String|ToString(Float) IntroTriggerPercent) " percent through the flight path.") true true "(R=0.000000,G=0.660000,B=1.000000,A=1.000000)" 8.0)
  (Variables|Default|SetDelayedTriggerIntroSequencePercent IntroTriggerPercent)
  (Actor|Tick|SetActorTickEnabled true))

(event EventTick
  (params (DeltaSeconds :type Float%20%28single-precision%29))
  (bind _returnvalue (BattleRoyaleGamePhaseLogic|GetBattleRoyaleGamePhaseLogicComponent))
  ;[0 0 0 0.5] For testing delaying the intro sequence to fake late joiners
  ;;[1 1 1] [Tick] - Only enabled when using the DelayIntroSequenceTrigger cheat.
  (bind _array_element (Utilities|Array|ForEachLoopwithBreak (Fortnite|GetLocalFortPlayerControllers true) :!label _label)
    (:LoopBody
      (if (>= (CallFunction|CalculateFlightPathProgressPercent _array_element) (Variables|Default|GetDelayedTriggerIntroSequencePercent))
        (Utilities|IsValid _returnvalue
          (:Is%20Valid
            (switch Utilities|FlowControl|Switch|SwitchonEAthenaGamePhase (BattleRoyaleGamePhaseLogic|GetGamePhase _returnvalue)
              (:Aircraft
                (Utilities|FlowControl|DoOnce false
                  (:Completed
                    (Development|PrintString "[Tick] Activating delayed intro sequence." true true "(R=0.000000,G=0.660000,B=1.000000,A=1.000000)" 8.0)
                    (CallFunction|BeginIntroSequence true)
                    (Actor|Tick|SetActorTickEnabled false)
                    (goto _label Break))))))
          (:Is%20Not%20Valid
            (Development|PrintString "[Tick] Cannot get GamePhase.  Unable to trigger delay intro sequence." true true "(R=0.000000,G=0.660000,B=1.000000,A=1.000000)" 8.0)))))))

(fn BindToOnExitAircraft
  (for _array_element (Fortnite|GetLocalFortPlayerControllers true)
    (Utilities|IsValid (Actor|GetComponentbyClass _array_element "/Script/FortniteGame.FortControllerComponent_Aircraft")
      (:Is%20Valid
        (EventDispatchers|BindEventtoOnExitAircraftEvent (Actor|GetComponentbyClass _array_element "/Script/FortniteGame.FortControllerComponent_Aircraft") (EventDispatchers|CreateEvent :!event OnExitAircraft))))))

(fn CalculateAspectRatio
  (bind (_returnvalue_x _returnvalue_y) (Math|Vector2D|BreakVector2D (Viewport|GetViewportSize)))
  (return (Math|Float|SafeDivide _returnvalue_x _returnvalue_y)))

(fn CalculateDistanceFromCamera
  (params (WildEstateAircraftRef :type Actor%20Object%20Reference%20%28by%20ref%29) (PlayerCameraManagerRef :type Player%20Camera%20Manager%20Object%20Reference%20%28by%20ref%29))
  (return (+ (Variables|Config|GetPlaneDistanceFromViewTarget) (Math|Vector|Distance(Vector) (Transformation|GetActorLocation WildEstateAircraftRef) (Camera|GetCameraLocation PlayerCameraManagerRef)))))

(fn CalculateFlightPathProgressPercent
  (params (PlayerController :type Fort%20Player%20Controller%20Object%20Reference))
  (bind _returnvalue (Pawn|GetViewTarget PlayerController))
  (Utilities|IsValid PlayerController
    (:Is%20Valid
      (bind _asfort_athena_aircraft (Utilities|Casting|CastToFortAthenaAircraft _returnvalue)
        (:then
          (return (Aircraft|GetTotalFlightTimePercentage _asfort_athena_aircraft)))
        (:CastFailed
          (Fortnite|FortShippingLog self (Utilities|String|Append "[CalculateFlightPathProgressPercent] ViewTarget is NOT FortAthenaAircraft.  [" (Utilities|GetDisplayName _returnvalue) "]") true))))
    (:Is%20Not%20Valid
      _))
  (return -1.0))

(fn CalculateHeightScale
  (params (WildEstateAircraftRef :type Actor%20Object%20Reference%20%28by%20ref%29) (PlayerCameraManagerRef :type Player%20Camera%20Manager%20Object%20Reference%20%28by%20ref%29))
  (return (Utilities|Operators|Multiply 2.0 (CallFunction|CalculateDistancefromCamera WildEstateAircraftRef PlayerCameraManagerRef) (Math|Trig|Tan(Radians) (Math|Trig|DegreesToRadians (/ (Camera|GetFOVAngle PlayerCameraManagerRef) 2.0))))))

(fn FitIntroCardToViewport
  (params (PlaneStaticMesh :type Static%20Mesh%20Component%20Object%20Reference%20%28by%20ref%29) (PlayerViewInfo_PlayerControllerRef_2_74C1CA9F4CE69EC3722FDE9EB6414FF8 :type Fort%20Player%20Controller%20Object%20Reference) (PlayerViewInfo_AircraftRef_17_F59A25BD418B6C826BF779B03F5B9E8C :type Fort%20Athena%20Aircraft%20Object%20Reference) (PlayerViewInfo_PlayerCameraManagerRef_9_B78DA68F490CBF5D4F6166B7A1088195 :type Player%20Camera%20Manager%20Object%20Reference) (PlayerViewInfo_IntroCardRef_12_023F7DB34C8B6B566E978D97B57377AB :type BP%20Intro%20Sequence%20Card%20Object%20Reference) (PlayerViewInfo_CameraModifierRef_15_4B58C77948E7321B314F7B84B4C31B53 :type Camera%20Modifier%20Wild%20Estate%20Intro%20Object%20Reference))
  (Transformation|SetRelativeScale3D PlaneStaticMesh (Math|Vector|MakeVector (/ (* (CallFunction|CalculateAspectRatio) (CallFunction|CalculateHeightScale PlayerViewInfo_AircraftRef_17_F59A25BD418B6C826BF779B03F5B9E8C PlayerViewInfo_PlayerCameraManagerRef_9_B78DA68F490CBF5D4F6166B7A1088195)) 1e+02) (/ (CallFunction|CalculateHeightScale PlayerViewInfo_AircraftRef_17_F59A25BD418B6C826BF779B03F5B9E8C PlayerViewInfo_PlayerCameraManagerRef_9_B78DA68F490CBF5D4F6166B7A1088195) 1e+02) 1.0)) ; Set the scale of the plane based on ViewHeight and ViewWidth
)

(fn GatherPlayerViewInfo
  (bind _self self)
  (for _array_element (Fortnite|GetLocalFortPlayerControllers true)
    (Utilities|IsValid _array_element
      (:Is%20Valid
        (Utilities|IsValid (Pawn|GetViewTarget _array_element)
          (:Is%20Valid
            (bind _asfort_athena_aircraft (Utilities|Casting|CastToFortAthenaAircraft (Pawn|GetViewTarget _array_element))
              (:then
                (Utilities|IsValid (Variables|/Script/Engine.PlayerController|GetPlayerCameraManager _array_element)
                  (:Is%20Valid
                    (Utilities|Array|Add (Variables|CachedRefs|GetPlayerViewInfos) (Utilities|Struct|MakeLocalPlayerControllerView _array_element _asfort_athena_aircraft (Variables|/Script/Engine.PlayerController|GetPlayerCameraManager _array_element))))
                  (:Is%20Not%20Valid
                    (Fortnite|FortShippingLog _self "[BP_Intro_WildEstate_Card] [GatherPlayerViewInfo] PlayerCameraManager is null" true))))
              (:CastFailed
                (Fortnite|FortShippingLog _self "[BP_Intro_WildEstate_Card] [GatherPlayerViewInfo] Failed cast to WildEstateAircraft" true))))
          (:Is%20Not%20Valid
            (Fortnite|FortShippingLog _self "[BP_Intro_WildEstate_Card] [GatherPlayerViewInfo] ViewTarget is null" true))))
      (:Is%20Not%20Valid
        (Fortnite|FortShippingLog _self "[BP_Intro_WildEstate_Card] [GatherPlayerViewInfo] Controller is null" true)))))

(fn SetUpCameraModifier
  (params (PlayerViewInfo :type Local%20Player%20Controller%20View%20Structure%20%28by%20ref%29))
  (bind (_playercontrollerref_2_74c1ca9f4ce69ec3722fde9eb6414ff8 _aircraftref_17_f59a25bd418b6c826bf779b03f5b9e8c _playercameramanagerref_9_b78da68f490cbf5d4f6166b7a1088195) (Utilities|Struct|BreakLocalPlayerControllerView PlayerViewInfo))
  (bind _returnvalue (Game|Player|AddNewCameraModifier _playercameramanagerref_9_b78da68f490cbf5d4f6166b7a1088195 "/WildEstate/TechArt/Intro/CameraModifier_WildEstate_Intro.CameraModifier_WildEstate_Intro_C"))
  (bind _structout (Utilities|Struct|SetmembersinLocalPlayerControllerView PlayerViewInfo _returnvalue))
  (bind (_structout_playercontrollerref_2_74c1ca9f4ce69ec3722fde9eb6414ff8 _structout_aircraftref_17_f59a25bd418b6c826bf779b03f5b9e8c _structout_playercameramanagerref_9_b78da68f490cbf5d4f6166b7a1088195 _structout_introcardref_12_023f7db34c8b6b566e978d97b57377ab _structout_cameramodifierref_15_4b58c77948e7321b314f7b84b4c31b53) (Utilities|Struct|BreakLocalPlayerControllerView _structout))
  (Variables|/WildEstate/TechArt/Intro/CameraModifier_WildEstate_Intro.CameraModifier_WildEstate_Intro_C|SetAircraftFlightRotator (Aircraft|GetFlightStartRotation _structout_aircraftref_17_f59a25bd418b6c826bf779b03f5b9e8c) _structout_cameramodifierref_15_4b58c77948e7321b314f7b84b4c31b53)
  (return PlayerViewInfo))

(fn SetUpIntroCard
  (params (PlayerViewInfo :type Local%20Player%20Controller%20View%20Structure%20%28by%20ref%29))
  (bind _playerviewinfo (Variables|GetPlayerViewInfo))
  (bind _returnvalue (Game|SpawnActorfromClass "/WildEstate/TechArt/Intro/BP_IntroSequenceCard.BP_IntroSequenceCard_C" (struct :Location "0, 0, 0" :Rotation "0, 0, 0" :Scale "1.000000,1.000000,1.000000") "AlwaysSpawn" "MultiplyWithRoot" self 0))
  (bind _structout (Utilities|Struct|SetmembersinLocalPlayerControllerView PlayerViewInfo _returnvalue))
  (bind (_structout_playercontrollerref_2_74c1ca9f4ce69ec3722fde9eb6414ff8 _structout_aircraftref_17_f59a25bd418b6c826bf779b03f5b9e8c _structout_playercameramanagerref_9_b78da68f490cbf5d4f6166b7a1088195 _structout_introcardref_12_023f7db34c8b6b566e978d97b57377ab) (Utilities|Struct|BreakLocalPlayerControllerView _structout))
  (Transformation|AttachActorToActor _structout_introcardref_12_023f7db34c8b6b566e978d97b57377ab _structout_aircraftref_17_f59a25bd418b6c826bf779b03f5b9e8c "None" "KeepRelative" "KeepRelative" "KeepRelative" true)
  (Transformation|SetRelativeLocation (Variables|/WildEstate/TechArt/Intro/BP_IntroSequenceCard.BP_IntroSequenceCard_C|GetPlane _returnvalue) (Math|Vector|MakeVector (Variables|Config|GetPlaneDistanceFromViewTarget)) false false)
  (Transformation|SetRelativeRotation (Variables|/WildEstate/TechArt/Intro/BP_IntroSequenceCard.BP_IntroSequenceCard_C|GetPlane _returnvalue) (Variables|Config|GetPlaneRelativeRotation) false false)
  (Default|FitIntroCardtoViewport (Variables|/WildEstate/TechArt/Intro/BP_IntroSequenceCard.BP_IntroSequenceCard_C|GetPlane _returnvalue) _playerviewinfo)
  (Rendering|SetCastShadow (Variables|/WildEstate/TechArt/Intro/BP_IntroSequenceCard.BP_IntroSequenceCard_C|GetPlane _returnvalue) false)
  (return _playerviewinfo))

(fn UnbindFromOnExitAircraft
  (for _array_element (Fortnite|GetLocalFortPlayerControllers true)
    (Utilities|IsValid (Actor|GetComponentbyClass _array_element "/Script/FortniteGame.FortControllerComponent_Aircraft")
      (:Is%20Valid
        (EventDispatchers|UnbindEventfromOnExitAircraftEvent (Actor|GetComponentbyClass _array_element "/Script/FortniteGame.FortControllerComponent_Aircraft") (EventDispatchers|CreateEvent :!event OnExitAircraft))))))

(fn ConstructionScript)
