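; Camera modifier for the WildEstate intro that applies relative location and rotation offsets to the view, leaving FOV unchanged.
; The location offset is rotated by the aircraft flight rotator before being added to the view location.
; The view forward vector is pitched around its right axis, then yawed around world Z.
; A world-up-referenced up vector is rolled around that forward, and MakeRotFromZX rebuilds the final rotation, avoiding direct rotator addition.
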(fn BlueprintModifyCamera
  (params (DeltaTime :type Float%20%28single-precision%29) (ViewLocation :type Vector :default "0, 0, 0") (ViewRotation :type Rotator :default "0, 0, 0") (FOV :type Float%20%28single-precision%29))
  (bind (_camerarotationoffsetrelative_roll _camerarotationoffsetrelative_pitch _camerarotationoffsetrelative_yaw) (Math|Rotator|BreakRotator (Variables|GetCameraRotationOffsetRelative_Roll)))
  (bind (_camerarotationoffsetrelative_roll_2 _camerarotationoffsetrelative_pitch_2) (Math|Rotator|BreakRotator (Variables|GetCameraRotationOffsetRelative_Roll)))
  (bind _camerarotationoffsetrelative_roll_3 (Math|Rotator|BreakRotator (Variables|GetCameraRotationOffsetRelative_Roll)))
  (return (+ ViewLocation (Math|Transform|TransformDirection (Math|Transform|MakeTransform "0, 0, 0" (Variables|Default|GetAircraftFlightRotator)) (Variables|Default|GetCameraLocationOffsetRelative))) (Math|Rotator|MakeRotfromZX (Math|Vector|RotateVectorAroundAxis (Math|Vector|GetUpVector (Math|Rotator|MakeRotfromXZ (Math|Vector|RotateVectorAroundAxis (Math|Vector|RotateVectorAroundAxis (Math|Vector|GetForwardVector ViewRotation) _camerarotationoffsetrelative_pitch_2 (Math|Vector|GetRightVector ViewRotation)) _camerarotationoffsetrelative_yaw "0.000000,0.000000,1.000000") (0.0 0.0 1.0))) _camerarotationoffsetrelative_roll_3 (Math|Vector|RotateVectorAroundAxis (Math|Vector|RotateVectorAroundAxis (Math|Vector|GetForwardVector ViewRotation) _camerarotationoffsetrelative_pitch_2 (Math|Vector|GetRightVector ViewRotation)) _camerarotationoffsetrelative_yaw "0.000000,0.000000,1.000000")) (Math|Vector|RotateVectorAroundAxis (Math|Vector|RotateVectorAroundAxis (Math|Vector|GetForwardVector ViewRotation) _camerarotationoffsetrelative_pitch_2 (Math|Vector|GetRightVector ViewRotation)) _camerarotationoffsetrelative_yaw "0.000000,0.000000,1.000000")) FOV))
