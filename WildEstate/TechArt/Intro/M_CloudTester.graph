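# Unlit test material that previews one cloud from the Wild Estate intro cloud flipbook, selected by CloudIndex.
# Five rows are packed in each of the red and green channels, with the index choosing row and channel.
# V is scaled and clamped around the center before sampling.
# CloudMin/CloudMax remap the value into a curve-atlas color ramp, multiplied by a separately remapped brightness mask.
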Material /WildEstate/TechArt/Intro/M_CloudTester
graph {
 properties {bAutomaticallySetUsageInEditor=false, shadingModel=MSM_Unlit, usedShadingModels=Unlit}
 nodes [
  Constant() {r=5}
  n1 = ScalarParameter() {defaultValue=6, parameterName=CloudIndex}
  n2 = Constant() {r=5}
  n3 = Floor(None=Divide(A=@n1, B=@n2))
  LinearInterpolate(A=Constant3Vector() {constant=<rgba 1 0 0 0>}, B=Constant3Vector() {constant=<rgba 0 1 0 0>}, Alpha=@n3)
  n4 = Modulo!(A=@n1, B=@n2)
  n5 = MaterialFunctionCall(Float2=TextureCoordinate()) {materialFunction=&BreakOutFloat2Components}
  n6 = Subtract(A=Multiply(A=@n5.G, B=@n2), B=Divide(A=@n2))
  n7 = MaterialFunctionCall(
   "Animation  Phase (0-1)"=Divide(A=Floor(None=@n4), B=@n2)
   Number of Rows=@n2
   Number of Columns=Constant() {r=1}
   Texture=TextureObject() {texture=/WildEstate/TechArt/Intro/Textures/T_WildEstate_Intro_CloudFlipbook}
   UVs=Saturate(None=AppendVector(A=@n5.R, B=@n6))
  ) {materialFunction=&FlipBook}
  n8 = If(A=@n3, "A > B"=ComponentMask(None=@n7.Result) {g=true}, "A < B"=ComponentMask(None=@n7.Result) {r=true}) {constB=1}
  n9 = InverseLinearInterpolate(A=ScalarParameter() {defaultValue=0.45, parameterName=CloudMin}, B=ScalarParameter() {defaultValue=0.675, parameterName=CloudMax}, Value=@n8)
  n10 = InverseLinearInterpolate(Value=@n8) {constB=0.4}
  Multiply(
   A=CurveAtlasRowParameter(CurveTime=Saturate(None=@n9)) {atlas=&/CA_IntroClouds, curve=&/C_IntroCloudColors}
   B=Saturate(None=Power(Base=@n10) {constExponent=1})
  )
 ]
}
