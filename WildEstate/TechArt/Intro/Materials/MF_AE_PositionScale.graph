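# Material function that converts After Effects position and scale values for use in materials, under the Intro tech art folder.
# Position is divided by a project resolution parameter and remapped to a centred -1 to 1 range, with a width-over-width factor that leaves it unscaled.
# Scale is converted from percent and multiplied by a hardcoded constant.
# A separate calculation from the resolution height feeds no output.
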MaterialFunction /WildEstate/TechArt/Intro/Materials/MF_AE_PositionScale
graph {nodes [
  n1 = VectorParameter() {group=AE_PositionScale, parameterName=AE Project Resolution, defaultValue=<rgba 2560 1440 0 1>}
  MaterialExpressionConstantBiasScale_0 = ConstantBiasScale(None=Divide(
    A=FunctionInput() {bUsePreviewValueAsDefault=true, inputName=AE Position, inputType=FunctionInput_Vector2}
    B=ComponentMask(None=@n1.RGB) {g=true, r=true}
   )) {bias=-0.5, scale=2}
  n2 = Convert(None=@MaterialExpressionConstantBiasScale_0) {
   nodeName=Break Float2
   convertInputs [{
     type=Vector2
     expressionInput {
      expression=&self:MaterialExpressionConstantBiasScale_0
      inputName=None
      mask=0
      maskA=0
      maskB=0
      maskG=0
      maskR=0
      outputIndex=0
     }
    }]
   convertOutputs [{type=Scalar}, {type=Scalar}]
  }
  n3 = Divide(A=@n1.R, B=@n1.R)
  FunctionOutput(None=AppendVector(A=Multiply(A=@n2, B=@n3), B=@n2)) {outputName=Position Out}
  n4 = FunctionInput() {bUsePreviewValueAsDefault=true, inputName=AE Scale, inputType=FunctionInput_Vector2}
  FunctionOutput(None=Multiply(A=Divide(A=@n4) {constB=100}, B=Constant() {r=4.2389})) {outputName=Scale Out}
  Multiply(A=Divide(A=Constant() {r=6104}, B=Multiply(A=@n1.G) {constB=2})) {constB=2}
 ]}
