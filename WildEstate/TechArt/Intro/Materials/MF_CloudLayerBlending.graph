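# MF_CloudLayerBlending is a material function in the Intro tech art set that blends a cloud color layer over a base color.
# It multiplies Cloud Color by the scalar Cloud Alpha and Base Color by one minus that alpha, then adds the results into Color Out.
# The math is a manual linear interpolation, equivalent to a standard alpha-over blend with non-premultiplied cloud color.
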MaterialFunction /WildEstate/TechArt/Intro/Materials/MF_CloudLayerBlending
graph {nodes [
  n1 = FunctionInput() {inputName=Cloud Color}
  n2 = FunctionInput() {inputName=Cloud Alpha, inputType=FunctionInput_Scalar, sortPriority=1}
  n3 = FunctionInput() {inputName=Base Color, sortPriority=2}
  n4 = OneMinus(None=@n2)
  FunctionOutput(None=Add(A=Multiply(A=@n1, B=@n2), B=Multiply(A=@n3, B=@n4))) {outputName=Color Out}
 ]}
