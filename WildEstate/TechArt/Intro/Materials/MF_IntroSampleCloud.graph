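# Material function that samples one intro cloud from a packed, single-column flipbook texture and outputs its color, alpha and a pass-through depth.
# Cloud Index modulo the row count picks the row, while index divided by rows selects the RG or BA channel pair, doubling capacity.
# The first channel is remapped by min/max parameters into a curve atlas gradient for color.
# The second is remapped by falloff inputs and raised to a power for alpha.
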MaterialFunction /WildEstate/TechArt/Intro/Materials/MF_IntroSampleCloud
graph {nodes [
  n1 = FunctionInput(Preview=Constant() {r=5}) {bUsePreviewValueAsDefault=true, inputName=Cloud Texture Rows}
  n2 = FunctionInput() {inputName=Cloud Index, inputType=FunctionInput_Scalar, sortPriority=1}
  n3 = Modulo!(A=@n2, B=@n1)
  n4 = MaterialFunctionCall(
   "Animation  Phase (0-1)"=Divide(A=Floor(None=@n3), B=@n1)
   Number of Rows=@n1
   Number of Columns=Constant() {r=1}
   Texture=FunctionInput() {inputName=Cloud Texture, inputType=FunctionInput_Texture2D}
   UVs=FunctionInput() {inputName=UV In, inputType=FunctionInput_Vector2, sortPriority=2}
  ) {materialFunction=&FlipBook}
  n5 = If(
   A=Floor(None=Divide(A=@n2, B=@n1))
   "A > B"=AppendVector(A=ComponentMask(None=@n4.Result) {b=true}, B=@n4.Alpha)
   "A < B"=ComponentMask(None=@n4.Result) {g=true, r=true}
  ) {constB=1}
  n6 = InverseLinearInterpolate(
   A=ScalarParameter() {defaultValue=0.05, group=Intro WildEstate Clouds, parameterName=CloudColorMin}
   B=ScalarParameter() {defaultValue=1, group=Intro WildEstate Clouds, parameterName=CloudColorMax}
   Value=ComponentMask(None=@n5) {r=true}
  )
  n7 = CurveAtlasRowParameter(CurveTime=Saturate(None=@n6)) {
   atlas=/WildEstate/TechArt/Intro/CA_IntroClouds
   curve=/WildEstate/TechArt/Intro/C_IntroCloudColors_Corrected
   defaultValue=1
   group=Intro WildEstate Colors
   parameterName=CloudColorCurve
  }
  FunctionOutput(None=@n7) {outputName=Color}
  n8 = InverseLinearInterpolate(
   A=Max(A=FunctionInput() {inputName=Alpha Falloff Min, inputType=FunctionInput_Scalar, sortPriority=3}) {constB=0.06}
   B=FunctionInput() {inputName=Alpha Falloff Max, inputType=FunctionInput_Scalar, sortPriority=4}
   Value=ComponentMask(None=@n5) {g=true}
  )
  FunctionOutput(None=Power(Base=Saturate(None=@n8), Exp=FunctionInput() {inputName=Alpha Power, inputType=FunctionInput_Scalar, sortPriority=5})) {outputName=Alpha, sortPriority=1}
  FunctionOutput(None=FunctionInput() {inputName=Depth, inputType=FunctionInput_Scalar, sortPriority=6}) {outputName=Depth, sortPriority=2}
  If(A=@n6, "A > B"=Constant3Vector() {constant=<rgba 1 0 0 0>}, "A < B"=@n7) {constB=1}
  If(A=@n6, "A > B"=@n7, "A < B"=Constant3Vector() {constant=<rgba 0 1 0 0>})
  n9 = MaterialFunctionCall(Float2=TextureCoordinate()) {materialFunction=&BreakOutFloat2Components}
  n10 = Multiply(A=@n9.G, B=@n1)
  n11 = Divide(A=@n1)
  Saturate(None=AppendVector(A=@n9.R, B=Subtract(A=@n10, B=@n11)))
 ]}
