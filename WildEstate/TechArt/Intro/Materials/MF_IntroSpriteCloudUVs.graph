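# Material function that computes animated UVs and depth for a cloud sprite in the intro, placed in square, centered screen space.
# Animation Progress, raised to an optional Progress Power, drives linear blends between starting and ending position, scale and depth through a named reroute.
# Screen UVs are offset by the position, divided by Scale Ratio times scale, then bias-scaled and saturated to clamp the UV output.
# The blended depth is returned as a second output.
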MaterialFunction /WildEstate/TechArt/Intro/Materials/MF_IntroSpriteCloudUVs
graph {nodes [
  n1 = FunctionInput() {inputName=Starting Position, inputType=FunctionInput_Vector2, sortPriority=4}
  n2 = FunctionInput() {inputName=Ending Position, inputType=FunctionInput_Vector2, sortPriority=5}
  n3 = NamedRerouteUsage()
  n4 = FunctionInput() {inputName=Scale Ratio, inputType=FunctionInput_Vector2, sortPriority=1}
  n5 = LinearInterpolate(
   A=FunctionInput() {inputName=Starting Scale, inputType=FunctionInput_Scalar, sortPriority=2}
   B=FunctionInput() {inputName=Ending Scale, inputType=FunctionInput_Scalar, sortPriority=3}
   Alpha=NamedRerouteUsage()
  )
  n6 = Multiply(
   A=Subtract(A=FunctionInput() {inputName=Square Centered Screen UVs, inputType=FunctionInput_Vector2}, B=LinearInterpolate(A=@n1, B=@n2, Alpha=@n3))
   B=Divide(B=Multiply(A=@n4, B=@n5))
  )
  FunctionOutput(None=Saturate(None=ConstantBiasScale(None=@n6))) {outputName=UVs}
  n7 = Constant()
  n8 = Constant()
  FunctionOutput(None=LinearInterpolate(
    A=FunctionInput(Preview=@n7) {bUsePreviewValueAsDefault=true, inputName=Starting Depth, inputType=FunctionInput_Scalar, sortPriority=7}
    B=FunctionInput(Preview=@n8) {bUsePreviewValueAsDefault=true, inputName=Ending Depth, inputType=FunctionInput_Scalar, sortPriority=8}
    Alpha=NamedRerouteUsage()
   )) {outputName=Depth, sortPriority=1}
  n9 = Constant() {r=1}
  NamedRerouteDeclaration(None=Power(
    Base=FunctionInput() {inputName=Animation Progress, sortPriority=9}
    Exp=FunctionInput(Preview=@n9) {bUsePreviewValueAsDefault=true, inputName=Progress Power, inputType=FunctionInput_Scalar, sortPriority=10}
   )) {name=Animation Progress, nodeColor=<rgba 0.4531 0 1 1>}
 ]}
