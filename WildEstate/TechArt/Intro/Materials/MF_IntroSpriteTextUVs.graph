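# Material function that computes animated UVs and depth for a zooming text sprite in the intro.
# A shared progress value, remapped from a start offset, drives power-curved blends for scale, multiplied secondary scale, depth and zoom center.
# Square-centered screen UVs are offset, divided by the ratio-adjusted scale, biased and saturated.
# The zoom center eases in via smoothstep up to a cutoff, serving as both scaling pivot and UV offset.
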MaterialFunction /WildEstate/TechArt/Intro/Materials/MF_IntroSpriteTextUVs
graph {nodes [
  n1 = FunctionInput() {inputName=Square Centered Screen UVs, inputType=FunctionInput_Vector2}
  n2 = Saturate(None=FunctionInput() {inputName=Zoom Center Animation Strength, inputType=FunctionInput_Scalar, sortPriority=9})
  n3 = FunctionInput() {inputName=Zoom Center Animation Progress Cutoff, inputType=FunctionInput_Scalar, sortPriority=11}
  n4 = NamedRerouteUsage()
  n5 = Power(
   Base=Saturate(None=InverseLinearInterpolate(B=@n3, Value=@n4))
   Exp=FunctionInput() {inputName=Zoom Center Animation Power, inputType=FunctionInput_Scalar, sortPriority=10}
  )
  n6 = Multiply(
   A=FunctionInput() {inputName=Zoom Center, inputType=FunctionInput_Vector2, sortPriority=7}
   B=LinearInterpolate(A=OneMinus(None=@n2), Alpha=SmoothStep(Value=@n5))
  )
  n7 = Add(A=FunctionInput() {inputName=Starting Position, inputType=FunctionInput_Vector2, sortPriority=8}, B=@n6)
  n8 = FunctionInput() {inputName=Starting Scale, inputType=FunctionInput_Scalar, sortPriority=2}
  n9 = FunctionInput() {inputName=Ending Scale, inputType=FunctionInput_Scalar, sortPriority=3}
  n10 = NamedRerouteUsage()
  n11 = Power(Base=@n10, Exp=FunctionInput() {inputName=Scale Power, inputType=FunctionInput_Scalar, sortPriority=5})
  n12 = FunctionInput() {inputName=Secondary Ending Scale, inputType=FunctionInput_Scalar, sortPriority=4}
  n13 = Power(Base=@n10, Exp=FunctionInput() {inputName=Secondary Scale Power, inputType=FunctionInput_Scalar, sortPriority=6})
  n14 = Multiply(
   A=FunctionInput() {inputName=Scale Ratio, inputType=FunctionInput_Vector2, sortPriority=1}
   B=Multiply(A=LinearInterpolate(A=@n8, B=@n9, Alpha=@n11), B=LinearInterpolate(B=@n12, Alpha=@n13) {constA=1})
  )
  n15 = ConstantBiasScale(None=Multiply(A=Subtract(A=@n1, B=@n7), B=Divide(B=@n14)))
  FunctionOutput(None=Saturate(None=Add(A=@n15, B=@n6))) {outputName=UVs}
  n16 = Constant()
  n17 = Constant()
  n18 = NamedRerouteUsage()
  n19 = FunctionInput() {inputName=Depth Power, inputType=FunctionInput_Scalar, sortPriority=14}
  FunctionOutput(None=LinearInterpolate(
    A=FunctionInput(Preview=@n16) {bUsePreviewValueAsDefault=true, inputName=Starting Depth, inputType=FunctionInput_Scalar, sortPriority=12}
    B=FunctionInput(Preview=@n17) {bUsePreviewValueAsDefault=true, inputName=Ending Depth, inputType=FunctionInput_Scalar, sortPriority=13}
    Alpha=Power(Base=@n18, Exp=@n19)
   )) {outputName=Depth, sortPriority=1}
  n20 = FunctionInput() {inputName=Animation Progress Start Offset, inputType=FunctionInput_Scalar, sortPriority=16}
  n21 = FunctionInput() {inputName=Animation Progress, inputType=FunctionInput_Scalar, sortPriority=15}
  NamedRerouteDeclaration(None=Saturate(None=InverseLinearInterpolate(A=@n20, Value=@n21))) {name=Animation Progress, nodeColor=<rgba 0.4531 0 1 1>}
 ]}
