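# Unlit masked intro title card for WildEstate, animated by the Progress value in MPC_Intro_WildEstate.
# Ten flipbook cloud sprites, moved between AE-style position and scale keyframes, layer over a screen-space background.
# SDF title text with drop shadow and two-step noise wobble zooms toward a set center, depth-fading between cloud layers.
# A timed center-hole mask and ending fade accompany the zoom; color passes through emissive adjustments.
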Material /WildEstate/TechArt/Intro/Materials/M_IntroSequenceCard
graph {
 properties {
  bAutomaticallySetUsageInEditor=false
  bCastRayTracedShadows=false
  bDisablePreExposureScale=true
  bUseTranslucencyVertexFog=false
  blendMode=BLEND_Masked
  blendableLocation=BL_SceneColorBeforeBloom
  blendableOutputAlpha=true
  shadingModel=MSM_Unlit
  translucencyPass=MTP_AfterMotionBlur
  usedShadingModels=Unlit
 }
 nodes [
  Constant()
  n1 = Divide(A=Time(), B=Constant() {r=5.5})
  Frac(None=@n1)
  MaterialFunctionCall(Offset Amount=ScalarParameter() {group=Intro WildEstate, parameterName=Camera Offset Amount}) {materialFunction=/Engine/Functions/Engine_MaterialFunctions02/WorldPositionOffset/CameraOffset}
  n2 = VectorParameter() {group=Intro WildEstate Colors, parameterName=Color Out Multiplier, sortPriority=-1, defaultValue=<rgba 0.3323 0.3323 0.3323 1>}
  MaterialFunctionCall(Emissive=Multiply(A=NamedRerouteUsage(), B=@n2.RGB)) {materialFunction=&MF_FX_EmissiveAdjustments}
  n3 = InverseLinearInterpolate(A=ScalarParameter() {defaultValue=0.95, group=Intro WildEstate, parameterName=Ending Fade Start}, Value=NamedRerouteUsage()) {constB=0.99}
  Multiply(A=OneMinus(None=NamedRerouteUsage()), B=OneMinus(None=Saturate(None=@n3)))
  n4 = MaterialFunctionCall(Float2=ConstantBiasScale(None=ScreenPosition().ViewportUV) {bias=-0.5, scale=2}) {materialFunction=&BreakOutFloat2Components}
  n5 = MaterialFunctionCall(Float2=ViewSize()) {materialFunction=&BreakOutFloat2Components}
  n6 = Divide(A=@n5.R, B=@n5.R)
  NamedRerouteDeclaration(None=AppendVector(A=Multiply(A=@n4.R, B=@n6), B=@n4.G)) {name=Square Centered Screen UVs, nodeColor=<rgba 0 0.5078 1 1>}
  NamedRerouteDeclaration(None=Constant2Vector() {g=0.2, r=1}) {name=Cloud Scale Ratio, nodeColor=<rgba 0.8281 1 0 1>}
  n7 = MaterialFunctionCall!(AE Scale=Constant() {r=40.6124}, AE Position=Constant2Vector() {g=747.5555, r=1955.7778}) {materialFunction=&/MF_AE_PositionScale}
  n8 = MaterialFunctionCall(AE Scale=Constant() {r=84.5914}, AE Position=Constant2Vector() {g=432.127, r=4376.3491}) {materialFunction=&MF_AE_PositionScale}
  n9 = Constant() {r=0.475}
  n10 = MaterialFunctionCall!(
   Square Centered Screen UVs=NamedRerouteUsage()
   Scale Ratio=NamedRerouteUsage()
   Starting Scale=@n7.Scale Out
   Ending Scale=@n8.Scale Out
   Starting Position=@n7.Scale Out
   Ending Position=@n8.Scale Out
   Starting Depth=@n9
   Ending Depth=@n9
   Animation Progress=NamedRerouteUsage()
   Progress Power=NamedRerouteUsage()
  ) {materialFunction=&/MF_IntroSpriteCloudUVs}
  n11 = MaterialFunctionCall!(
   Cloud Texture=NamedRerouteUsage()
   Cloud Texture Rows
   Cloud Index=Constant() {r=2}
   UV In=@n10.UVs
   Alpha Falloff Min=Constant()
   Alpha Falloff Max=Constant() {r=0.4}
   Alpha Power=Constant() {r=1.5}
   Depth=@n10.UVs
  ) {materialFunction=&/MF_IntroSampleCloud}
  n12 = Subtract(A=@n11.Depth, B=NamedRerouteUsage())
  n13 = ScalarParameter() {defaultValue=0.05, group=Intro WildEstate, parameterName=Text Depth Fade Range}
  n14 = OneMinus(None=Saturate(None=Divide(A=@n12, B=@n13)))
  n15 = NamedRerouteUsage()
  n16 = MaterialFunctionCall(AE Scale=Constant() {r=72.451}, AE Position=Constant2Vector() {g=1276.2822, r=692}) {materialFunction=&MF_AE_PositionScale}
  n17 = MaterialFunctionCall(AE Scale=Constant() {r=128.451}, AE Position=Constant2Vector() {g=2030.568, r=-530.8571}) {materialFunction=&MF_AE_PositionScale}
  n18 = MaterialFunctionCall(
   Square Centered Screen UVs=NamedRerouteUsage()
   Scale Ratio=NamedRerouteUsage()
   Starting Scale=@n16.Scale Out
   Ending Scale=@n17.Scale Out
   Starting Position=@n16.Scale Out
   Ending Position=@n17.Scale Out
   Animation Progress=NamedRerouteUsage()
   Progress Power=NamedRerouteUsage()
  ) {materialFunction=&MF_IntroSpriteCloudUVs}
  n19 = MaterialFunctionCall(
   Cloud Texture=NamedRerouteUsage()
   Cloud Index=Constant()
   UV In=@n18.UVs
   Alpha Falloff Min=Constant()
   Alpha Falloff Max=Constant() {r=0.2}
   Alpha Power=Constant() {r=1}
   Depth=@n18.UVs
  ) {materialFunction=&MF_IntroSampleCloud}
  n20 = MaterialFunctionCall(AE Scale=Constant() {r=52}, AE Position=Constant2Vector() {g=25, r=1993.4287}) {materialFunction=&MF_AE_PositionScale}
  n21 = MaterialFunctionCall(AE Scale=Constant() {r=84}, AE Position=Constant2Vector() {g=-440.1429, r=3143.4287}) {materialFunction=&MF_AE_PositionScale}
  n22 = MaterialFunctionCall(
   Square Centered Screen UVs=NamedRerouteUsage()
   Scale Ratio=NamedRerouteUsage()
   Starting Scale=@n20.Scale Out
   Ending Scale=@n21.Scale Out
   Starting Position=@n20.Scale Out
   Ending Position=@n21.Scale Out
   Animation Progress=NamedRerouteUsage()
   Progress Power=NamedRerouteUsage()
  ) {materialFunction=&MF_IntroSpriteCloudUVs}
  n23 = MaterialFunctionCall(
   Cloud Texture=NamedRerouteUsage()
   Cloud Index=Constant() {r=5}
   UV In=@n22.UVs
   Alpha Falloff Min=Constant()
   Alpha Falloff Max=Constant() {r=0.5}
   Alpha Power=Constant() {r=1.5}
   Depth=@n22.UVs
  ) {materialFunction=&MF_IntroSampleCloud}
  n24 = MaterialFunctionCall(AE Scale=Constant() {r=53}, AE Position=Constant2Vector() {g=288, r=781.7143}) {materialFunction=&MF_AE_PositionScale}
  n25 = MaterialFunctionCall(AE Scale=Constant() {r=89}, AE Position=Constant2Vector() {g=-316.5714, r=-251.4286}) {materialFunction=&MF_AE_PositionScale}
  n26 = MaterialFunctionCall(
   Square Centered Screen UVs=NamedRerouteUsage()
   Scale Ratio=NamedRerouteUsage()
   Starting Scale=@n24.Scale Out
   Ending Scale=@n25.Scale Out
   Starting Position=@n24.Scale Out
   Ending Position=@n25.Scale Out
   Animation Progress=NamedRerouteUsage()
   Progress Power=NamedRerouteUsage()
  ) {materialFunction=&MF_IntroSpriteCloudUVs}
  n27 = MaterialFunctionCall(
   Cloud Texture=NamedRerouteUsage()
   Cloud Index=Constant() {r=1}
   UV In=@n26.UVs
   Alpha Falloff Min=Constant()
   Alpha Falloff Max=Constant() {r=0.35}
   Alpha Power=Constant() {r=2}
   Depth=@n26.UVs
  ) {materialFunction=&MF_IntroSampleCloud}
  n28 = MaterialFunctionCall(AE Scale=Constant() {r=59}, AE Position=Constant2Vector() {g=685.7143, r=630.8571}) {materialFunction=&MF_AE_PositionScale}
  n29 = MaterialFunctionCall(AE Scale=Constant() {r=103}, AE Position=Constant2Vector() {g=584.5714, r=-658.0317}) {materialFunction=&MF_AE_PositionScale}
  n30 = Constant() {r=0.35}
  n31 = MaterialFunctionCall(
   Square Centered Screen UVs=NamedRerouteUsage()
   Scale Ratio=NamedRerouteUsage()
   Starting Scale=@n28.Scale Out
   Ending Scale=@n29.Scale Out
   Starting Position=@n28.Scale Out
   Ending Position=@n29.Scale Out
   Starting Depth=@n30
   Ending Depth=@n30
   Animation Progress=NamedRerouteUsage()
   Progress Power=NamedRerouteUsage()
  ) {materialFunction=&MF_IntroSpriteCloudUVs}
  n32 = MaterialFunctionCall(
   Cloud Texture=NamedRerouteUsage()
   Cloud Index=Constant() {r=7}
   UV In=@n31.UVs
   Alpha Falloff Min=Constant()
   Alpha Falloff Max=Constant() {r=0.4}
   Alpha Power=Constant() {r=1.5}
   Depth=@n31.UVs
  ) {materialFunction=&MF_IntroSampleCloud}
  n33 = NamedRerouteUsage()
  n34 = Saturate(None=Divide(A=Subtract(A=@n32.Depth, B=@n33), B=ScalarParameter() {defaultValue=0.05, parameterName=Text Depth Fade Range}))
  n35 = MaterialFunctionCall(AE Scale=Constant() {r=53.9339}, AE Position=Constant2Vector() {g=978.3704, r=542.8148}) {materialFunction=&MF_AE_PositionScale}
  n36 = MaterialFunctionCall(AE Scale=Constant() {r=78}, AE Position=Constant2Vector() {g=1042.3704, r=-410.0741}) {materialFunction=&MF_AE_PositionScale}
  n37 = MaterialFunctionCall(
   Square Centered Screen UVs=NamedRerouteUsage()
   Scale Ratio=NamedRerouteUsage()
   Starting Scale=@n35.Scale Out
   Ending Scale=@n36.Scale Out
   Starting Position=@n35.Scale Out
   Ending Position=@n36.Scale Out
   Animation Progress=NamedRerouteUsage()
   Progress Power=NamedRerouteUsage()
  ) {materialFunction=&MF_IntroSpriteCloudUVs}
  n38 = MaterialFunctionCall(
   Cloud Texture=NamedRerouteUsage()
   Cloud Index=Constant() {r=6}
   UV In=@n37.UVs
   Alpha Falloff Min=Constant()
   Alpha Falloff Max=Constant() {r=0.5}
   Alpha Power=Constant() {r=1.5}
   Depth=@n37.UVs
  ) {materialFunction=&MF_IntroSampleCloud}
  n39 = MaterialFunctionCall(AE Scale=Constant() {r=58.2938}, AE Position=Constant2Vector() {g=1280, r=2168}) {materialFunction=&MF_AE_PositionScale}
  n40 = MaterialFunctionCall(AE Scale=Constant() {r=90}, AE Position=Constant2Vector() {g=1704.2963, r=3152.2964}) {materialFunction=&MF_AE_PositionScale}
  n41 = MaterialFunctionCall(
   Square Centered Screen UVs=NamedRerouteUsage()
   Scale Ratio=NamedRerouteUsage()
   Starting Scale=@n39.Scale Out
   Ending Scale=@n40.Scale Out
   Starting Position=@n39.Scale Out
   Ending Position=@n40.Scale Out
   Animation Progress=NamedRerouteUsage()
   Progress Power=NamedRerouteUsage()
  ) {materialFunction=&MF_IntroSpriteCloudUVs}
  n42 = MaterialFunctionCall(
   Cloud Texture=NamedRerouteUsage()
   Cloud Index=Constant() {r=3}
   UV In=@n41.UVs
   Alpha Falloff Min=Constant()
   Alpha Falloff Max=Constant() {r=0.275}
   Alpha Power=Constant() {r=2}
   Depth=@n41.UVs
  ) {materialFunction=&MF_IntroSampleCloud}
  n43 = MaterialFunctionCall(AE Scale=Constant() {r=28}, AE Position=Constant2Vector() {g=1050.0741, r=1716.1481}) {materialFunction=&MF_AE_PositionScale}
  n44 = MaterialFunctionCall(AE Scale=Constant() {r=40}, AE Position=Constant2Vector() {g=1201.7778, r=2159.4075}) {materialFunction=&MF_AE_PositionScale}
  n45 = MaterialFunctionCall(
   Square Centered Screen UVs=NamedRerouteUsage()
   Scale Ratio=NamedRerouteUsage()
   Starting Scale=@n43.Scale Out
   Ending Scale=@n44.Scale Out
   Starting Position=@n43.Scale Out
   Ending Position=@n44.Scale Out
   Animation Progress=NamedRerouteUsage()
   Progress Power=NamedRerouteUsage()
  ) {materialFunction=&MF_IntroSpriteCloudUVs}
  n46 = MaterialFunctionCall(
   Cloud Texture=NamedRerouteUsage()
   Cloud Index=Constant() {r=8}
   UV In=@n45.UVs
   Alpha Falloff Min=Constant()
   Alpha Falloff Max=Constant() {r=0.7}
   Alpha Power=Constant() {r=1.5}
   Depth=@n45.UVs
  ) {materialFunction=&MF_IntroSampleCloud}
  n47 = MaterialFunctionCall(AE Scale=Constant() {r=56}, AE Position=Constant2Vector() {g=363.4286, r=2443.4287}) {materialFunction=&MF_AE_PositionScale}
  n48 = MaterialFunctionCall(AE Scale=Constant() {r=80}, AE Position=Constant2Vector() {g=98.2857, r=3369.1428}) {materialFunction=&MF_AE_PositionScale}
  n49 = MaterialFunctionCall(
   Square Centered Screen UVs=NamedRerouteUsage()
   Scale Ratio=NamedRerouteUsage()
   Starting Scale=@n47.Scale Out
   Ending Scale=@n48.Scale Out
   Starting Position=@n47.Scale Out
   Ending Position=@n48.Scale Out
   Animation Progress=NamedRerouteUsage()
   Progress Power=NamedRerouteUsage()
  ) {materialFunction=&MF_IntroSpriteCloudUVs}
  n50 = MaterialFunctionCall(
   Cloud Texture=NamedRerouteUsage()
   Cloud Index=Constant() {r=9}
   UV In=@n49.UVs
   Alpha Falloff Min=Constant()
   Alpha Falloff Max=Constant() {r=0.5}
   Alpha Power=Constant() {r=2}
   Depth=@n49.UVs
  ) {materialFunction=&MF_IntroSampleCloud}
  n51 = MaterialFunctionCall(AE Scale=Constant() {r=41}, AE Position=Constant2Vector() {g=426.0741, r=1064.2963}) {materialFunction=&MF_AE_PositionScale}
  n52 = MaterialFunctionCall(AE Scale=Constant() {r=61}, AE Position=Constant2Vector() {g=194.2105, r=613.0135}) {materialFunction=&MF_AE_PositionScale}
  n53 = MaterialFunctionCall(
   Square Centered Screen UVs=NamedRerouteUsage()
   Scale Ratio=NamedRerouteUsage()
   Starting Scale=@n51.Scale Out
   Ending Scale=@n52.Scale Out
   Starting Position=@n51.Scale Out
   Ending Position=@n52.Scale Out
   Animation Progress=NamedRerouteUsage()
   Progress Power=NamedRerouteUsage()
  ) {materialFunction=&MF_IntroSpriteCloudUVs}
  n54 = MaterialFunctionCall(
   Cloud Texture=NamedRerouteUsage()
   Cloud Index=Constant() {r=4}
   UV In=@n53.UVs
   Alpha Falloff Min=Constant()
   Alpha Falloff Max=Constant() {r=0.6}
   Alpha Power=Constant() {r=1.5}
   Depth=@n53.UVs
  ) {materialFunction=&MF_IntroSampleCloud}
  n55 = VectorParameter() {group=Intro WildEstate Colors, parameterName=Background Color Multiplier + Desaturation, defaultValue=<rgba 1 1 1 0>}
  n56 = Multiply(
   A=TextureSample(UVs=ScreenPosition().ViewportUV, Tex=TextureObject() {texture=/WildEstate/TechArt/Intro/Textures/T_WildEstate_Intro_Background}) {texture=&T_WildEstate_Intro_Background}.RGB
   B=@n55.RGB
  )
  n57 = MaterialFunctionCall!(
   Cloud Color=@n54.Color
   Cloud Alpha=@n54.Color
   Base Color=MaterialFunctionCall(Cloud Color=NamedRerouteUsage(), Cloud Alpha=NamedRerouteUsage(), Base Color=Desaturation(None=@n56, Fraction=@n55.A)) {materialFunction=&/MF_CloudLayerBlending}.Color Out
  ) {materialFunction=&MF_CloudLayerBlending}
  n58 = MaterialFunctionCall(
   Cloud Color=@n42.Color
   Cloud Alpha=@n42.Color
   Base Color=MaterialFunctionCall(
    Cloud Color=@n46.Color
    Cloud Alpha=@n46.Color
    Base Color=MaterialFunctionCall(Cloud Color=@n50.Color, Cloud Alpha=@n50.Color, Base Color=@n57.Color Out) {materialFunction=&MF_CloudLayerBlending}.Color Out
   ) {materialFunction=&MF_CloudLayerBlending}.Color Out
  ) {materialFunction=&MF_CloudLayerBlending}
  n59 = MaterialFunctionCall(
   Cloud Color=NamedRerouteUsage()
   Cloud Alpha=Multiply(A=OneMinus(None=@n34), B=NamedRerouteUsage())
   Base Color=MaterialFunctionCall(
    Cloud Color=@n32.Color
    Cloud Alpha=@n32.Color
    Base Color=MaterialFunctionCall(Cloud Color=@n38.Color, Cloud Alpha=@n38.Color, Base Color=@n58.Color Out) {materialFunction=&MF_CloudLayerBlending}.Color Out
   ) {materialFunction=&MF_CloudLayerBlending}.Color Out
  ) {materialFunction=&MF_CloudLayerBlending}
  n60 = MaterialFunctionCall(
   Cloud Color=@n23.Color
   Cloud Alpha=@n23.Color
   Base Color=MaterialFunctionCall(
    Cloud Color=@n27.Color
    Cloud Alpha=@n27.Color
    Base Color=MaterialFunctionCall(Cloud Color=@n11.Color, Cloud Alpha=@n11.Color, Base Color=@n59.Color Out) {materialFunction=&MF_CloudLayerBlending}.Color Out
   ) {materialFunction=&MF_CloudLayerBlending}.Color Out
  ) {materialFunction=&MF_CloudLayerBlending}
  NamedRerouteDeclaration(None=MaterialFunctionCall(
    Cloud Color=NamedRerouteUsage()
    Cloud Alpha=Multiply(A=@n14, B=@n15)
    Base Color=MaterialFunctionCall(Cloud Color=@n19.Color, Cloud Alpha=@n19.Color, Base Color=@n60.Color Out) {materialFunction=&MF_CloudLayerBlending}.Color Out
   ) {materialFunction=&MF_CloudLayerBlending}.Color Out) {name=Cloud Blend Out, nodeColor=<rgba 0.3594 1 0 1>}
  NamedRerouteDeclaration(None=Constant() {r=1.3}) {name=Cloud Progress Power, nodeColor=<rgba 1 0 0.6797 1>}
  NamedRerouteDeclaration(None=CollectionParameter() {collection=/WildEstate/TechArt/Intro/MPC_Intro_WildEstate, parameterName=Progress}) {name=Progress, nodeColor=<rgba 1 0 0.7734 1>}
  NamedRerouteDeclaration(None=TextureObject() {samplerType=SAMPLERTYPE_LinearColor, texture=/WildEstate/TechArt/Intro/Textures/T_WildEstate_Intro_CloudFlipbook_IPArt_Linear}) {name=Cloud Texture, nodeColor=<rgba 0 1 0.3203 1>}
  n61 = ScalarParameter() {defaultValue=0.075, group=Intro WildEstate, parameterName=Text SDF Min}
  n62 = ComponentMask(None=VectorParameter() {group=Intro WildEstate, parameterName=ZoomCenterCoordinates, defaultValue=<rgba -0.2125 0.17 0 1>}.RGB) {g=true, r=true}
  n63 = MaterialFunctionCall!(
   Square Centered Screen UVs=NamedRerouteUsage()
   Scale Ratio=Constant() {r=1}
   Starting Scale=Constant() {r=0.175}
   Ending Scale=Constant() {r=10}
   Secondary Ending Scale=Constant() {r=4}
   Scale Power=Constant() {r=5}
   Secondary Scale Power=Constant() {r=5}
   Zoom Center=@n62
   Starting Position=ComponentMask(None=VectorParameter() {group=Intro WildEstate, parameterName=StartingPositionCoordinates, defaultValue=<rgba 0.02 -0.02 0 1>}.RGB) {g=true, r=true}
   Zoom Center Animation Strength=Constant() {r=1}
   Zoom Center Animation Power=Constant() {r=2}
   Zoom Center Animation Progress Cutoff=Constant() {r=0.8}
   Starting Depth=Constant()
   Ending Depth=Constant() {r=1}
   Depth Power=Constant() {r=1}
   Animation Progress=NamedRerouteUsage()
   Animation Progress Start Offset=Constant()
  ) {materialFunction=&/MF_IntroSpriteTextUVs}
  n64 = Time()
  n65 = ScalarParameter() {defaultValue=10, group=Intro WildEstate, parameterName=Text Noise Speed}
  n66 = Floor(None=Modulo!(A=Multiply(A=@n64, B=@n65), B=Constant() {r=2}))
  n67 = ScalarParameter() {defaultValue=8, group=Intro WildEstate, parameterName=Text Noise Scale}
  n68 = TextureSample(UVs=Add(A=Multiply(A=@n66) {constB=0.5}, B=Multiply(A=@n63.UVs, B=@n67))) {samplerType=SAMPLERTYPE_Grayscale, texture=/Game/Global/Textures/Default/Noise/LowResBlurredNoiseNormalized}
  n69 = Add(
   A=@n63.UVs
   B=Multiply(
    A=ConstantBiasScale(None=@n68.R) {bias=-0.5, scale=2}
    B=ScalarParameter() {defaultValue=0.004, group=Intro WildEstate, parameterName=Text Noise Strength}
   )
  )
  n70 = TextureObjectParameter() {group=Intro WildEstate, parameterName=TitleTexture, sortPriority=0, texture=/WildEstate/TechArt/Intro/Textures/br_text_sdf}
  n71 = Ceil(None=InverseLinearInterpolate(
    A=@n61
    Value=TextureSample(UVs=@n69, Tex=@n70) {samplerSource=SSM_Clamp_WorldGroupSettings, texture=/WildEstate/TechArt/Intro/Textures/WildEstate_IntroText}.R
   ))
  n72 = Add(
   A=@n69
   B=ComponentMask(None=VectorParameter() {group=Intro WildEstate, parameterName=Text Drop Shadow Offset, defaultValue=<rgba 0.005 -0.005 0 1>}.RGB) {g=true, r=true}
  )
  n73 = Ceil(None=InverseLinearInterpolate(
    A=@n61
    Value=TextureSample(UVs=@n72, Tex=@n70) {samplerSource=SSM_Clamp_WorldGroupSettings, texture=&WildEstate_IntroText}.R
   ))
  n74 = Saturate(None=Add(A=@n71, B=@n73))
  n75 = OneMinus(None=@n74)
  n76 = SphereMask(A=ConstantBiasScale(None=@n63.UVs) {bias=-0.5, scale=2}, B=Multiply(A=@n62) {constB=2}) {attenuationRadius=0.13}
  n77 = InverseLinearInterpolate(
   A=ScalarParameter() {defaultValue=0.465, group=Intro WildEstate, parameterName=Hole Fade Start}
   B=ScalarParameter() {defaultValue=0.6, group=Intro WildEstate, parameterName=Hole Fade End}
   Value=NamedRerouteUsage()
  ) {constA=0.7, constB=0.8}
  NamedRerouteDeclaration(None=Multiply(A=Multiply(A=@n75, B=@n76), B=Saturate(None=@n77))) {name=Center Mask, nodeColor=<rgba 0.3828 0 1 1>}
  NamedRerouteDeclaration(None=@n63.Depth) {name=Text Depth, nodeColor=<rgba 1 0 0.2578 1>}
  n78 = InverseLinearInterpolate(
   A=ScalarParameter() {defaultValue=0.1, group=Intro WildEstate, parameterName=Text Fade Start}
   B=ScalarParameter() {defaultValue=0.15, group=Intro WildEstate, parameterName=Text Fade End}
   Value=NamedRerouteUsage()
  )
  NamedRerouteDeclaration(None=Multiply(A=@n74, B=Saturate(None=@n78))) {name=Text Alpha, nodeColor=<rgba 1 0 0.2578 1>}
  n79 = VectorParameter() {group=Intro WildEstate Colors, parameterName=Text Color, defaultValue=<rgba 0.6867 0.4397 0.0018 1>}
  n80 = Multiply(A=OneMinus(None=@n71), B=@n73)
  n81 = VectorParameter() {group=Intro WildEstate Colors, parameterName=Text Shadow Color, defaultValue=<rgba 0.1356 0.1301 0.1301 1>}
  NamedRerouteDeclaration(None=Add(A=Multiply(A=@n71, B=@n79.RGB), B=Multiply(A=@n80, B=@n81.RGB))) {name=Text Color, nodeColor=<rgba 1 0 0.2578 1>}
  n82 = ArcsineFast!(None=Sine(None=@n1) {period=2})
  Saturate(None=ConstantBiasScale(None=Divide(A=@n82) {constB=1.5708}))
 ]
}
